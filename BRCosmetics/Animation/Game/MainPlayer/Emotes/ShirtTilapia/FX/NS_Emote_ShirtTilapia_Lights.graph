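# Niagara effect for the ShirtTilapia emote that shows pulsing red lights on a skeletal mesh's spheres.
# Six timed bursts per emitter use spawn group as bone/socket index to attach particles.
# Two mirrored emitters face opposite ±Y sides, fading sprites and glow by camera-to-system-Y dot product.
# Waveforms pulse brightness and time of day tints sprites; only the first emitter adds light renderers, scaled by time of day.
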NiagaraSystem /BRCosmetics/Animation/Game/MainPlayer/Emotes/ShirtTilapia/FX/NS_Emote_ShirtTilapia_Lights
dependencies {
 usedDataInterfaces [&NiagaraDataInterfaceCamera, &NiagaraDataInterfaceSkeletalMesh]
 usedDynamicInputs [&FNTimeOfDayColor]
 usedModules [
  &self:SystemSpawnScript.SystemScriptSource.SystemScriptGraph.NiagaraNodeAssignment_0.SetVariables_#1
  &self:SystemSpawnScript.SystemScriptSource.SystemScriptGraph.NiagaraNodeAssignment_1.SetVariables_#2
  &SystemState
 ]
}
emitters {
 Lights {
  emitterUpdateScript [
   &EmitterState(Life Cycle Mode=::NewEnumerator1, Loop Behavior=::NewEnumerator1, Loop Duration Mode=::NewEnumerator1, UseLoopDelay=false)
   &SpawnBurst_Instantaneous(Use Spawn Probability=false, UseLoopCountLimit=false)
   &SpawnBurst_Instantaneous(Spawn Group=1, Spawn Time=0.18, Use Spawn Probability=false, UseLoopCountLimit=false) {module=SpawnBurst_Instantaneous001}
   &SpawnBurst_Instantaneous(Spawn Group=2, Spawn Time=0.28, Use Spawn Probability=false, UseLoopCountLimit=false) {module=SpawnBurst_Instantaneous002}
   &SpawnBurst_Instantaneous(Spawn Group=3, Spawn Time=1.77, Use Spawn Probability=false, UseLoopCountLimit=false) {module=SpawnBurst_Instantaneous003}
   &SpawnBurst_Instantaneous(Spawn Group=4, Spawn Time=1.88, Use Spawn Probability=false, UseLoopCountLimit=false) {module=SpawnBurst_Instantaneous004}
   &SpawnBurst_Instantaneous(Spawn Group=5, Spawn Time=2.05, Use Spawn Probability=false, UseLoopCountLimit=false) {module=SpawnBurst_Instantaneous005}
  ]
  particleSpawnScript [
   &InitializeParticle(
    Color=&MultiplyLinearColor(Linear Color=<rgba 5 0 0 1>, Scale Factor=&MakeVector4_FromLinearColor(LinearColor=$System.ToD))
    Lifetime=5
    Sprite Size Mode=::NewEnumerator3
    Uniform Sprite Size=7.5
    UsePositionOffset=false
    Write Material Random=true
   )
   &CameraOffset(Camera Offset Amount=15)
   &SpriteFacingAndAlignment(Sprite Facing=<0 1 0>, WriteAlignment=false, WriteFacing=true)
  ]
  particleUpdateScript [
   &ParticleState(Kill Particles When Lifetime Has Elapsed=false, Let Infinitely Lived Particles Die When Emitter Deactivates=true)
   &SkeletalMeshLocation(
    Bone / Socket Index=&SpawnGroup()
    Bone Sampling Mode=::NewEnumerator4
    MaskSpawnGroup=false
    Sampled Position Offset=<0 2 0>
    Skeletal Mesh=$System.T Spheres
   )
   &ScaleColor(
    Scale Alpha=&Waveform(
     "Waveform [1]"=::NewEnumerator0
     "[1] Frequency"=&RandomRangeFloat(Minimum=1, Maximum=1.5)
     "[1] Amplitude Min/Max"=<0 1>
    )
    ScaleA=true
    ScaleRGB=true
   )
   &ScaleColor(
    Scale Alpha=&Lerp_Float(
     A=0
     B=1
     Alpha=&Max_Float(
      A:NiagaraFloat=0
      B:NiagaraFloat=&Multiply_Float(
       A=&DotProduct(
        Vector A=&NormalizeVector(
         Vector To Normalize=&ReturnCameraVector(Camera Vectors=::NewEnumerator1, Coordinate Space=::World)
         Safe Normalize=true
         Safe Fallback Threshold=0
         Safe Fallback Vector=<0 0 0.001>
        )
        Vector B=&NormalizeVector(Vector To Normalize=$Engine.Owner.SystemYAxis, Safe Normalize=true, Safe Fallback Threshold=0, Safe Fallback Vector=<0 0 0.001>)
       )
       B=-1
      )
     )
    )
    ScaleA=true
    ScaleRGB=true
   ) {module=ScaleColor001}
   /SetVariables(
    Particles.GlowColor=&MakeLinearColorFromVectorAndFloat(
     "Vector (RGB)"=&MakeVectorFromLinearColorRGB(Color=<rgba 1 0 0 1>)
     "Float (Alpha)"=&Lerp_Float(
      A=0
      B=0.25
      Alpha=&Max_Float(
       A:NiagaraFloat=0
       B:NiagaraFloat=&Multiply_Float(
        A=&DotProduct(
         Vector A=&NormalizeVector(
          Vector To Normalize=&ReturnCameraVector(Camera Vectors=::NewEnumerator1, Coordinate Space=::World)
          Safe Normalize=true
          Safe Fallback Threshold=0
          Safe Fallback Vector=<0 0 0.001>
         )
         Vector B=&NormalizeVector(Vector To Normalize=$Engine.Owner.SystemYAxis, Safe Normalize=true, Safe Fallback Threshold=0, Safe Fallback Vector=<0 0 0.001>)
        )
        B=-1
       )
      )
     )
    )
    Particles.GlowSize=&Vector2DFromFloat(Value=5.5)
   ) {isSetParameters=true}
   &FN_Light_Attributes(
    Color=<rgba 1 0 0 1>
    Create Color Attribute=true
    Create Exponent Attribute=true
    Create Radius Attribute=true
    Exponent=6
    Radius=18
    Scale Color=&Waveform(
     "Waveform [1]"=::NewEnumerator0
     "[1] Frequency"=&RandomRangeFloat(Minimum=0.8, Maximum=1)
     "[1] Amplitude Min/Max"=<0.6 1>
     "[1] Scale"=&FNTimeOfDayScalar(Night Scale=2.5, Morning Scale=2.5, Day Scale=4, Evening Scale=2.5)
    )
   )
  ]
  rendererClasses [&NiagaraSpriteRendererProperties, &NiagaraLightRendererProperties]
  renderers [
   {rendererClass=&NiagaraSpriteRendererProperties}
   {rendererClass=&NiagaraLightRendererProperties, rendererIndex=1}
   {rendererClass=&NiagaraSpriteRendererProperties, rendererIndex=2}
  ]
 }
 Lights001 {
  emitterUpdateScript [
   &EmitterState(Life Cycle Mode=::NewEnumerator1, Loop Behavior=::NewEnumerator1, Loop Duration Mode=::NewEnumerator1, UseLoopDelay=false)
   &SpawnBurst_Instantaneous(Use Spawn Probability=false, UseLoopCountLimit=false)
   &SpawnBurst_Instantaneous(Spawn Group=1, Spawn Time=0.18, Use Spawn Probability=false, UseLoopCountLimit=false) {module=SpawnBurst_Instantaneous001}
   &SpawnBurst_Instantaneous(Spawn Group=2, Spawn Time=0.28, Use Spawn Probability=false, UseLoopCountLimit=false) {module=SpawnBurst_Instantaneous002}
   &SpawnBurst_Instantaneous(Spawn Group=3, Spawn Time=1.77, Use Spawn Probability=false, UseLoopCountLimit=false) {module=SpawnBurst_Instantaneous003}
   &SpawnBurst_Instantaneous(Spawn Group=4, Spawn Time=1.88, Use Spawn Probability=false, UseLoopCountLimit=false) {module=SpawnBurst_Instantaneous004}
   &SpawnBurst_Instantaneous(Spawn Group=5, Spawn Time=2.05, Use Spawn Probability=false, UseLoopCountLimit=false) {module=SpawnBurst_Instantaneous005}
  ]
  particleSpawnScript [
   &InitializeParticle(
    Color=&MultiplyLinearColor(Linear Color=<rgba 5 0 0 1>, Scale Factor=&MakeVector4_FromLinearColor(LinearColor=$System.ToD))
    Lifetime=5
    Sprite Size Mode=::NewEnumerator3
    Uniform Sprite Size=7.5
    UsePositionOffset=false
    Write Material Random=true
   )
   &CameraOffset(Camera Offset Amount=15)
   &SpriteFacingAndAlignment(Sprite Facing=<0 -1 0>, WriteAlignment=false, WriteFacing=true)
  ]
  particleUpdateScript [
   &ParticleState(Kill Particles When Lifetime Has Elapsed=false, Let Infinitely Lived Particles Die When Emitter Deactivates=true)
   &SkeletalMeshLocation(
    Bone / Socket Index=&SpawnGroup()
    Bone Sampling Mode=::NewEnumerator4
    MaskSpawnGroup=false
    Sampled Position Offset=<0 -2 0>
    Skeletal Mesh=$System.T Spheres
   )
   &ScaleColor(
    Scale Alpha=&Multiply_Float(
     A=&Waveform(
      Global Phase=$System.Age
      "Waveform [1]"=::NewEnumerator0
      "[1] Frequency"=&RandomRangeFloat(Minimum=1, Maximum=1.5)
      "[1] Amplitude Min/Max"=<0 1>
     )
     B=&Lerp_Float(
      A=0
      B=1
      Alpha=&Max_Float(
       A:NiagaraFloat=0
       B:NiagaraFloat=&Multiply_Float(
        A=&DotProduct(
         Vector A=&NormalizeVector(
          Vector To Normalize=&ReturnCameraVector(Camera Vectors=::NewEnumerator1, Coordinate Space=::World)
          Safe Normalize=true
          Safe Fallback Threshold=0
          Safe Fallback Vector=<0 0 0.001>
         )
         Vector B=&NormalizeVector(Vector To Normalize=$Engine.Owner.SystemYAxis, Safe Normalize=true, Safe Fallback Threshold=0, Safe Fallback Vector=<0 0 0.001>)
        )
        B=1
       )
      )
     )
    )
    ScaleA=true
    ScaleRGB=true
   ) {enabled=false}
   &ScaleColor(
    Scale Alpha=&Lerp_Float(
     A=0
     B=1
     Alpha=&Max_Float(
      A:NiagaraFloat=0
      B:NiagaraFloat=&Multiply_Float(
       A=&DotProduct(
        Vector A=&NormalizeVector(
         Vector To Normalize=&ReturnCameraVector(Camera Vectors=::NewEnumerator1, Coordinate Space=::World)
         Safe Normalize=true
         Safe Fallback Threshold=0
         Safe Fallback Vector=<0 0 0.001>
        )
        Vector B=&NormalizeVector(Vector To Normalize=$Engine.Owner.SystemYAxis, Safe Normalize=true, Safe Fallback Threshold=0, Safe Fallback Vector=<0 0 0.001>)
       )
       B=1
      )
     )
    )
    ScaleA=true
    ScaleRGB=true
   ) {module=ScaleColor001}
   &ScaleColor(
    Scale Alpha=&Waveform(
     "Waveform [1]"=::NewEnumerator0
     "[1] Frequency"=&RandomRangeFloat(Minimum=1, Maximum=1.5)
     "[1] Amplitude Min/Max"=<0 1>
    )
    ScaleA=true
    ScaleRGB=true
   ) {module=ScaleColor002}
   /SetVariables(
    Particles.GlowColor=&MakeLinearColorFromVectorAndFloat(
     "Vector (RGB)"=&MakeVectorFromLinearColorRGB(Color=<rgba 1 0 0 1>)
     "Float (Alpha)"=&Lerp_Float(
      A=0
      B=0.25
      Alpha=&Max_Float(
       A:NiagaraFloat=0
       B:NiagaraFloat=&Multiply_Float(
        A=&DotProduct(
         Vector A=&NormalizeVector(
          Vector To Normalize=&ReturnCameraVector(Camera Vectors=::NewEnumerator1, Coordinate Space=::World)
          Safe Normalize=true
          Safe Fallback Threshold=0
          Safe Fallback Vector=<0 0 0.001>
         )
         Vector B=&NormalizeVector(Vector To Normalize=$Engine.Owner.SystemYAxis, Safe Normalize=true, Safe Fallback Threshold=0, Safe Fallback Vector=<0 0 0.001>)
        )
        B=1
       )
      )
     )
    )
    Particles.GlowSize=&Vector2DFromFloat(Value=5.5)
   ) {isSetParameters=true}
  ]
  rendererClasses [&NiagaraSpriteRendererProperties]
  renderers [
   {rendererClass=&NiagaraSpriteRendererProperties}
   {rendererClass=&NiagaraSpriteRendererProperties, rendererIndex=1}
  ]
 }
}
summary {
 systemName=NS_Emote_ShirtTilapia_Lights
 emitters [
  {emitterName=Lights, rendererClasses [&NiagaraSpriteRendererProperties, &NiagaraLightRendererProperties]}
  {emitterName=Lights001, rendererClasses [&NiagaraSpriteRendererProperties]}
 ]
 userVariables []
}
