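# Niagara ribbon effect for the ShirtTilapia emote, drawing two short-lived animation trails from bone/socket data in one FN_AnimTrail emitter with two ribbon renderers.
# Spawn-group masks give each trail its own width, color and lifetime, and only the passthrough group tapers in width.
# Color is multiplied by a time-of-day value, and alpha follows a falling curve indexed by system age rather than particle age.
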NiagaraSystem /BRCosmetics/Animation/Game/MainPlayer/Emotes/ShirtTilapia/FX/NS_Emote_ShirtTilapia_Trail3
dependencies {
 usedDataInterfaces [&NiagaraDataInterfaceCurve, &NiagaraDataInterfaceSocketReader]
 usedDynamicInputs [&FNTimeOfDayColor]
 usedModules [
  &self:SystemSpawnScript.SystemScriptSource.SystemScriptGraph.NiagaraNodeAssignment_1.SetVariables_#1
  &SystemState
 ]
}
emitters {FN_AnimTrail {
  emitterSpawnScript [&FN_RibbonIDOffsetUtility()]
  emitterUpdateScript [
   &EmitterState()
   &FN_SpawnAnimTrail(Number of Anim Trails=2)
  ]
  particleSpawnScript [
   &FN_InitializeAnimTrailBoneIndecies()
   &FN_InitializeAnimTrailBoneIndecies(AnimTrail Selection=1) {module=FN_InitializeAnimTrailBoneIndecies001}
   &FN_InitializeAnimTrail(
    AnimTrailWidth=&MaskFloatBySpawnGroup(Masked Float=1.05, Passthrough Float=0.55)
    Color=&MultiplyLinearColor(
     Linear Color=&MaskLinearColorBySpawnGroup(Masked Linear Color=<rgba 0.0885 0.0885 0.0885 1>, Passthrough Linear Color=<rgba 0.3281 0.0161 0.012 1>)
     Scale Factor=&MakeVector4_FromLinearColor(LinearColor=$System.ToD)
    )
    Lifetime=&MaskFloatBySpawnGroup(Masked Float=0.12, Passthrough Float=0.16)
   )
   &AddVelocity(Velocity=<0 400 0>) {enabled=false}
   &CameraOffset(Camera Offset Amount=15)
  ]
  particleUpdateScript [
   &SolveForcesAndVelocity(Clamp Velocity=false, Limit Acceleration=false)
   &FN_UpdateAnimTrail() {module=FN_UpdateAnimTrail001}
   &ScaleColor(
    Scale Alpha=&FloatFromCurve(
     FloatCurve=DataInterface() {propertyValues {Curve {
        defaultValue=1
        keys [<at 0 1 RCIM_Cubic>, {arriveTangent=-2.61, interpMode=RCIM_Cubic, leaveTangent=-2.61, tangentMode=RCTM_User, time=1, value=0}]
       }}}
     CurveIndex=$System.Age
     Scale Curve=0.3
    )
    ScaleA=true
    ScaleRGB=true
   )
   &ScaleRibbonWidth(Ribbon Width Scale=&MaskFloatBySpawnGroup(
     Masked Float=1
     Passthrough Float=&FloatFromCurve(FloatCurve=DataInterface() {propertyValues {Curve {keys [<at 0 1 RCIM_Cubic>, {arriveTangent=-0.9774, interpMode=RCIM_Cubic, leaveTangent=-0.9774, tangentMode=RCTM_User, time=1, value=0}]}}})
    ))
  ]
  rendererClasses [&NiagaraRibbonRendererProperties]
  renderers [
   {rendererClass=&NiagaraRibbonRendererProperties}
   {rendererClass=&NiagaraRibbonRendererProperties, rendererIndex=1}
  ]
 }}
summary {
 systemName=NS_Emote_ShirtTilapia_Trail3
 emitters [{emitterName=FN_AnimTrail, rendererClasses [&NiagaraRibbonRendererProperties]}]
 userVariables [User.RibbonIDOffset {type=NiagaraInt32}]
}
